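# Unlit translucent glow material for the Veiled pickaxe, usable on Niagara mesh particles, with a purple HDR color default fed through an emissive adjustment function.
# An axis-aligned Fresnel on instance-space Z drives a faint center term and a reflected-gradient edge band with separate inner and outer exponents.
# The two terms are screen-blended into a glow mask.
# A vertex offset combines puffing along normals with a push toward the camera.
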Material /BRCosmetics/Weapons/FORT_Melee/Pickaxe_Veiled/FX/M_Pickaxe_Veiled_Glow
graph {
 properties {
  bUsedWithNiagaraMeshParticles=true
  blendMode=BLEND_Translucent
  pixelDepthOffsetMode=PDOM_Legacy
  shadingModel=MSM_Unlit
  usedShadingModels=Unlit
 }
 nodes [
  n1 = Subtract(A=CameraPositionWS(), B=WorldPosition().XYZ)
  Add(
   A=Multiply(A=VertexNormalWS(), B=ScalarParameter() {defaultValue=0.05, parameterName=Vertex Puff})
   B=Multiply(A=Normalize(VectorInput=@n1), B=ScalarParameter() {defaultValue=0.1, parameterName=Camera Offset})
  )
  MaterialFunctionCall(Emissive=VectorParameter() {parameterName=Color, defaultValue=<rgba 3.6 0.9 9 0.1>}.RGB) {materialFunction=&MF_FX_EmissiveAdjustments}
  n2 = MaterialFunctionCall(Axis=Transform(None=Constant3Vector() {constant=<rgba 0 0 1 0>}) {transformSourceType=TRANSFORMSOURCE_Instance}) {materialFunction=&AxisAlignedFresnel}
  n3 = Multiply(
   A=Power(Base=@n2.Fresnel, Exp=ScalarParameter() {defaultValue=2, parameterName=Center Exponent})
   B=Saturate(None=ScalarParameter() {defaultValue=0.03, parameterName=Center Value})
  )
  n4 = ScalarParameter() {defaultValue=0.333, parameterName=Outer Edge Size}
  n5 = Floor(None=Divide(A=@n2.Fresnel, B=@n4))
  n6 = Power(
   Base=MaterialFunctionCall!(Gradient=@n2.Fresnel, Center Point=@n4) {materialFunction=/Game/Effects/Fort_Effects/MaterialFunctions/MF_GradientReflected}.Result
   Exp=LinearInterpolate(
    A=ScalarParameter() {defaultValue=9, parameterName=Outer Fresnel Exponent}
    B=ScalarParameter() {defaultValue=3, parameterName=Inner Fresnel Exponent}
    Alpha=Saturate(None=@n5)
   )
  )
  OneMinus(None=Multiply(A=OneMinus(None=@n3), B=OneMinus(None=@n6)))
 ]
}
