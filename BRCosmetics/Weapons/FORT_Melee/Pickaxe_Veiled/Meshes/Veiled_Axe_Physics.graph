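# Physics asset for the Veiled pickaxe melee weapon mesh, defining collision and constraints for its dangling tassel.
# A capsule on the two-handed handle bone anchors a chain of six small tassel capsules, each constrained to the previous bone.
# All joints share the same limits, with wide swing and tight twist, so the tassel can bend along its length without much spinning.
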PhysicsAsset /BRCosmetics/Weapons/FORT_Melee/Pickaxe_Veiled/Meshes/Veiled_Axe_Physics
bodies {
 Melee_TwoHanded_Handle [{shapeType=Capsule, length=25.0671, radius=3.1788, center=<0 0 6.2918>}]
 dyn_tassel_1 [{shapeType=Capsule, length=0.25, radius=1, shapeName=dyn_tassel_1_capsule, center=<0.3985 0 0>, rotation=<rot 90 0 0>}]
 dyn_tassel_2 [{shapeType=Capsule, length=1, radius=1, shapeName=dyn_tassel_2_capsule, center=<1.6933 0 0>, rotation=<rot 90 0 0>}]
 dyn_tassel_3 [{shapeType=Capsule, length=1.25, radius=1, shapeName=dyn_tassel_3_capsule, center=<2.531 0 0>, rotation=<rot 90 0 0>}]
 dyn_tassel_4 [{shapeType=Capsule, length=1.25, radius=1, shapeName=dyn_tassel_4_capsule, center=<2.3251 0 0>, rotation=<rot 90 0 0>}]
 dyn_tassel_5 [{shapeType=Capsule, length=1.25, radius=1, shapeName=dyn_tassel_5_capsule, center=<1.9078 0 0>, rotation=<rot 90 0 0>}]
 dyn_tassel_6 [{shapeType=Capsule, length=0.25, radius=1, shapeName=dyn_tassel_6_capsule, center=<1.0577 0 0>, rotation=<rot 90 0 0>}]
}
constraints [
 {bone1Name=dyn_tassel_1, bone2Name=Melee_TwoHanded_Handle, swing1LimitDegrees=50, swing2LimitDegrees=50, twistLimitDegrees=5}
 {bone1Name=dyn_tassel_2, bone2Name=dyn_tassel_1, swing1LimitDegrees=50, swing2LimitDegrees=50, twistLimitDegrees=5}
 {bone1Name=dyn_tassel_3, bone2Name=dyn_tassel_2, swing1LimitDegrees=50, swing2LimitDegrees=50, twistLimitDegrees=5}
 {bone1Name=dyn_tassel_4, bone2Name=dyn_tassel_3, swing1LimitDegrees=50, swing2LimitDegrees=50, twistLimitDegrees=5}
 {bone1Name=dyn_tassel_5, bone2Name=dyn_tassel_4, swing1LimitDegrees=50, swing2LimitDegrees=50, twistLimitDegrees=5}
 {bone1Name=dyn_tassel_6, bone2Name=dyn_tassel_5, swing1LimitDegrees=50, swing2LimitDegrees=50, twistLimitDegrees=5}
]
